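# Sound cue for grooming foley of large, nail-less quadruped cosmetic companions, routed to the pets sound class.
# A quality-level root feeds the first two tiers an equal-weight random pick of four grooming waves, and the last two tiers a pick of only the first two waves.
# It uses companion foley attenuation, zero priority and shares the quadruped footstep concurrency set.
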SoundCue /CosmeticCompanions/Assets/Audio_Shared/Foley/Quadraped/Paws_NoNails/Large/SC_Mimosa_Quadruped_Paws_Lrg_Grooming
root=QualityLevel_0
nodes [
 {id=QualityLevel_0, children [Random_2,Random_2,Random_3,Random_3]}
 {id=Random_2, children [WavePlayer_6,WavePlayer_7,WavePlayer_8,WavePlayer_9], properties {weights [1,1,1,1]}}
 {id=Random_3, children [WavePlayer_10, WavePlayer_11], properties {weights [1,1]}}
 {
  id=WavePlayer_10
  properties {soundWaveAssetPtr=/CosmeticCompanions/Assets/Audio_Shared/Foley/Quadraped/Paws_NoNails/Large/Wavs/Mimosa_Paws_Lrg_Fol_Grooming_01}
 }
 {
  id=WavePlayer_11
  properties {soundWaveAssetPtr=/CosmeticCompanions/Assets/Audio_Shared/Foley/Quadraped/Paws_NoNails/Large/Wavs/Mimosa_Paws_Lrg_Fol_Grooming_02}
 }
 {
  id=WavePlayer_6
  properties {soundWaveAssetPtr=&Mimosa_Paws_Lrg_Fol_Grooming_01}
 }
 {
  id=WavePlayer_7
  properties {soundWaveAssetPtr=&Mimosa_Paws_Lrg_Fol_Grooming_02}
 }
 {
  id=WavePlayer_8
  properties {soundWaveAssetPtr=/CosmeticCompanions/Assets/Audio_Shared/Foley/Quadraped/Paws_NoNails/Large/Wavs/Mimosa_Paws_Lrg_Fol_Grooming_03}
 }
 {
  id=WavePlayer_9
  properties {soundWaveAssetPtr=/CosmeticCompanions/Assets/Audio_Shared/Foley/Quadraped/Paws_NoNails/Large/Wavs/Mimosa_Paws_Lrg_Fol_Grooming_04}
 }
]
properties {
 attenuationSettings=/Game/Sounds/Attenuation/AS_Companion_Foley
 maxDistance=1200
 priority=0
 soundClassObject=&Fort_Creatures_Pets
 concurrencySet [/CosmeticCompanions/Assets/Audio_Shared/Foley/Quadraped/CONC_Mimosa_Quadraped_Footstep]
}
