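# Sound cue for soft footsteps of large, nail-less quadruped companion paws, using walk recordings in the pets sound class with companion foley attenuation.
# A quality-level switch feeds the first two levels a random pick from all twelve walk waves and the other two a smaller random set of the first four.
# A top modulator randomly varies pitch and volume downward, atop reduced cue-wide pitch and volume, with a shared quadruped footstep concurrency limiting playback.
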SoundCue /CosmeticCompanions/Assets/Audio_Shared/Foley/Quadraped/Paws_NoNails/Large/SC_Mimosa_Quadruped_Paws_Lrg_NoNails_Footsteps_Soft
root=Modulator_0
nodes [
 {id=Modulator_0, children [QualityLevel_1], properties {pitchMax=1, pitchMin=0.8, volumeMax=1, volumeMin=0.6}}
 {id=QualityLevel_1, children [Random_0,Random_0,Random_4,Random_4]}
 {
  id=Random_0
  children [
   WavePlayer_0
   WavePlayer_1
   WavePlayer_2
   WavePlayer_3
   WavePlayer_4
   WavePlayer_5
   WavePlayer_6
   WavePlayer_7
   WavePlayer_8
   WavePlayer_9
   WavePlayer_10
   WavePlayer_11
  ]
  properties {weights [1,1,1,1,1,1,1,1,1,1,1,1]}
 }
 {id=Random_4, children [WavePlayer_12, WavePlayer_13, WavePlayer_28, WavePlayer_29], properties {weights [1,1,1,1]}}
 {
  id=WavePlayer_0
  properties {soundWaveAssetPtr=/CosmeticCompanions/Assets/Audio_Shared/Foley/Quadraped/Paws_NoNails/Wavs/Mimosa_Quadruped_Paws_Lrg_Walk_NoNails_12}
 }
 {
  id=WavePlayer_1
  properties {soundWaveAssetPtr=/CosmeticCompanions/Assets/Audio_Shared/Foley/Quadraped/Paws_NoNails/Wavs/Mimosa_Quadruped_Paws_Lrg_Walk_NoNails_11}
 }
 {
  id=WavePlayer_10
  properties {soundWaveAssetPtr=/CosmeticCompanions/Assets/Audio_Shared/Foley/Quadraped/Paws_NoNails/Wavs/Mimosa_Quadruped_Paws_Lrg_Walk_NoNails_02}
 }
 {
  id=WavePlayer_11
  properties {soundWaveAssetPtr=/CosmeticCompanions/Assets/Audio_Shared/Foley/Quadraped/Paws_NoNails/Wavs/Mimosa_Quadruped_Paws_Lrg_Walk_NoNails_01}
 }
 {
  id=WavePlayer_12
  properties {soundWaveAssetPtr=&Mimosa_Quadruped_Paws_Lrg_Walk_NoNails_01}
 }
 {
  id=WavePlayer_13
  properties {soundWaveAssetPtr=&Mimosa_Quadruped_Paws_Lrg_Walk_NoNails_02}
 }
 {
  id=WavePlayer_2
  properties {soundWaveAssetPtr=/CosmeticCompanions/Assets/Audio_Shared/Foley/Quadraped/Paws_NoNails/Wavs/Mimosa_Quadruped_Paws_Lrg_Walk_NoNails_10}
 }
 {
  id=WavePlayer_28
  properties {soundWaveAssetPtr=/CosmeticCompanions/Assets/Audio_Shared/Foley/Quadraped/Paws_NoNails/Wavs/Mimosa_Quadruped_Paws_Lrg_Walk_NoNails_03}
 }
 {
  id=WavePlayer_29
  properties {soundWaveAssetPtr=/CosmeticCompanions/Assets/Audio_Shared/Foley/Quadraped/Paws_NoNails/Wavs/Mimosa_Quadruped_Paws_Lrg_Walk_NoNails_04}
 }
 {
  id=WavePlayer_3
  properties {soundWaveAssetPtr=/CosmeticCompanions/Assets/Audio_Shared/Foley/Quadraped/Paws_NoNails/Wavs/Mimosa_Quadruped_Paws_Lrg_Walk_NoNails_09}
 }
 {
  id=WavePlayer_4
  properties {soundWaveAssetPtr=/CosmeticCompanions/Assets/Audio_Shared/Foley/Quadraped/Paws_NoNails/Wavs/Mimosa_Quadruped_Paws_Lrg_Walk_NoNails_08}
 }
 {
  id=WavePlayer_5
  properties {soundWaveAssetPtr=/CosmeticCompanions/Assets/Audio_Shared/Foley/Quadraped/Paws_NoNails/Wavs/Mimosa_Quadruped_Paws_Lrg_Walk_NoNails_07}
 }
 {
  id=WavePlayer_6
  properties {soundWaveAssetPtr=/CosmeticCompanions/Assets/Audio_Shared/Foley/Quadraped/Paws_NoNails/Wavs/Mimosa_Quadruped_Paws_Lrg_Walk_NoNails_06}
 }
 {
  id=WavePlayer_7
  properties {soundWaveAssetPtr=/CosmeticCompanions/Assets/Audio_Shared/Foley/Quadraped/Paws_NoNails/Wavs/Mimosa_Quadruped_Paws_Lrg_Walk_NoNails_05}
 }
 {
  id=WavePlayer_8
  properties {soundWaveAssetPtr=&Mimosa_Quadruped_Paws_Lrg_Walk_NoNails_04}
 }
 {
  id=WavePlayer_9
  properties {soundWaveAssetPtr=&Mimosa_Quadruped_Paws_Lrg_Walk_NoNails_03}
 }
]
properties {
 attenuationSettings=/Game/Sounds/Attenuation/AS_Companion_Foley
 maxDistance=1200
 pitchMultiplier=0.8
 priority=0
 soundClassObject=&Fort_Creatures_Pets
 volumeMultiplier=0.4
 concurrencySet [/CosmeticCompanions/Assets/Audio_Shared/Foley/Quadraped/CONC_Mimosa_Quadraped_Footstep]
}
